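# Conversation database for the Dreamflower NPC, defining its greeting and the branches it links into.
# The entry point leads to one Speech node with a default line plus prioritized contextual variants for time since the last match, first-ever conversation, and Bars held.
# That speech fans out through link nodes to shared entry tags for Season 17 Week 1, 3 and 13, basic services, and NPC library quests.
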ConversationDatabase /NPCLibrary/NPCs/Dreamflower/Dreamflower/NPCLibrary_Conversation_Dreamflower
entry_points [{tag=Conversation.Entry.Dreamflower, destinations [#1]}]
nodes [
 ConversationEntryPointNode_7 {
  id=#1
  class=ConversationEntryPointNode
  outputs [#2]
  properties {entryTag {tagName=Conversation.Entry.Dreamflower}}
 }
 Speech_C_4 {
  id=#2
  class=Speech_C
  outputs [
   #3
   #4
   #5
   #6
   #7
  ]
  properties {
   nodeName=""
   generalConfig {
    defaultMessage="Like, whoa, why so <b>hostile</>?"
    contextualMessages [
     {
      message="Wow, it's really you! I, like, totally had a dream about this."
      contextRequirements [{requirement {class=FortControllerRequirement_CheckTimeSinceLastPlayedMatch}}]
      priority {curve {rowName=Tandem.Speech.Priority.WhileSincePlayed}}
     }
     {
      message="Another child of the universe! Do you wanna, like, pick daisies and share vibes?"
      contextRequirements [{requirement {class=FortControllerRequirement_IsFirstEverConversationWithNPC}}]
      priority {curve {rowName=Tandem.Speech.Priority.FirstEverInteraction}}
     }
     {
      message="Oh, I'm not, like, into Bars. But I'm glad you've got lots!"
      contextRequirements [{requirement {class=FortControllerRequirement_HasBars, properties {thresholdTestConfig {requirement=GreaterThanOrEqual}}}}]
      priority {curve {rowName=Tandem.Speech.Priority.HighTotalBars}}
     }
    ]
   }
  }
 }
 ConversationLinkNode_4 {id=#7, class=ConversationLinkNode, properties {remoteEntryTag {tagName=Conversation.Entry.Season17.Week13}}}
 ConversationLinkNode_3 {id=#6, class=ConversationLinkNode, properties {remoteEntryTag {tagName=Conversation.Entry.Season17.Week3}}}
 ConversationLinkNode_0 {id=#5, class=ConversationLinkNode, properties {remoteEntryTag {tagName=Conversation.Entry.BasicServices}}}
 ConversationLinkNode_1 {id=#4, class=ConversationLinkNode, properties {remoteEntryTag {tagName=Conversation.Entry.NPCLibrary.Quests}}}
 ConversationLinkNode_2 {id=#3, class=ConversationLinkNode, properties {remoteEntryTag {tagName=Conversation.Entry.Season17.Week1}}}
]
